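# Niagara system for the FeatherMudLounge emote's SwapToLion loop, layering gold chest effects with mostly blue energy around the character.
# ChestLogo, small sparks and a camera-offset decal sprite form the focal burst, while embers and speed lines stream along negative X.
# Mesh ribbons and partitioned left/right eye trails follow the system skeletal mesh each update, with head trails at fixed offsets.
# Curves over loop age drive several spawn rates; ember glow alpha varies between gameplay and front end.
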NiagaraSystem /BRCosmetics/Animation/Game/MainPlayer/Emotes/FeatherMudLounge/FX/NS_Emote_FeatherMudLounge_SwapToLion_Loop
dependencies {
 usedDataInterfaces [
  &NiagaraDataInterfaceCurve
  &NiagaraDataInterfaceColorCurve
  &NiagaraDataInterfaceArrayFloat3
  &NiagaraDataInterfaceMeshRendererInfo
  &NiagaraDataInterfaceVectorCurve
  &NiagaraDataInterfaceCamera
  &NiagaraDataInterfaceSkeletalMesh
 ]
 usedDynamicInputs []
 usedModules [
  &self:SystemSpawnScript.SystemScriptSource.SystemScriptGraph.NiagaraNodeAssignment_1.SetVariables_#1
  &SystemState
 ]
}
emitters {
 ChestLogo {
  emitterUpdateScript [
   &EmitterState()
   &SpawnRate(
    SpawnGroup=1
    SpawnRate=&FloatFromCurve(
     FloatCurve=DataInterface() {propertyValues {Curve {keys [{interpMode=RCIM_Cubic, tangentMode=RCTM_Break, time=0, value=1}, {arriveTangent=-3.2993, tangentMode=RCTM_User, time=0.8, value=0}]}}}
     CurveIndex=$Emitter.NormalizedLoopAge
     Scale Curve=55
    )
    Use Spawn Probability=false
   )
   &SpawnBurst_Instantaneous(Use Spawn Probability=false, UseLoopCountLimit=false)
  ]
  particleSpawnScript [
   /SetVariables(Particles.VisibilityTag=&SpawnGroup()) {isSetParameters=true}
   &InitializeParticle(
    Color=&MaskLinearColorBySpawnGroup(
     Masked Linear Color=<rgba 3 2.2917 0.5889 0.7>
     Passthrough Linear Color=&RandomRangeLinearColor(Minimum=<rgba 5 2.5636 0 1>, Maximum=<rgba 6 4.7626 1.0503 1>)
    )
    Lifetime=&MaskFloatBySpawnGroup(Masked Float=1.4, Passthrough Float=&RandomRangeFloat(Minimum=0.4, Maximum=0.6))
    Position Offset=<22 0 17>
    Sprite Size=&MaskVector2dBySpawnGroup(
     Masked Vector2D=<23 23>
     Passthrough Vector2D=&RandomRangeVector2D(Minimum=<0.4 1>, Maximum=<1 3>)
     Spawn Group Mask=0
    )
    Sprite Size Mode=::NewEnumerator4
    UsePositionOffset=true
    Write Material Random=true
   )
   &ShapeLocation(
    MaskSpawnGroup=false
    Override Seed=false
    Sphere Radius=&MaskFloatBySpawnGroup(Masked Float=0, Passthrough Float=4)
    UseOwnerScale=true
   ) {enabled=false}
   &AddVelocity(
    Module.Use Falloff Distance=false
    Origin Offset=<22 0 17>
    Override Seed=false
    Velocity Mode=::NewEnumerator1
    Velocity Speed=&MaskFloatBySpawnGroup(Masked Float=0, Passthrough Float=&RandomRangeFloat(Minimum=25, Maximum=30))
   )
   &SpriteFacingAndAlignment(
    Facing Coordinate Space=::Local
    Sprite Alignment=&MaskVectorBySpawnGroup(
     Masked Vector=<0 0 1>
     Passthrough Vector=&NormalizeVector(Vector To Normalize=$Particles.Velocity, Safe Normalize=true, Safe Fallback Threshold=0, Safe Fallback Vector=<0 0 0.001>)
    )
    Sprite Facing=&MaskVectorBySpawnGroup(
     Masked Vector=<1 0 0>
     Passthrough Vector=&ReturnCameraVector(Camera Vectors=::NewEnumerator4, Coordinate Space=::Simulation, Origin Position=$Particles.Position)
    )
    WriteAlignment=true
    WriteFacing=true
   ) {enabled=false}
   /SetVariables(Particles.NewAlignment:Vector3f=<0 0 1>, Particles.NewFacing:Vector3f=<1 0 0>) {isSetParameters=true}
  ]
  particleUpdateScript [
   &ParticleState()
   &ScaleSpriteSize(Uniform Curve Sprite Scale=DataInterface() {propertyValues {Curve {keys [
        {interpMode=RCIM_Cubic, leaveTangent=1.8427, tangentMode=RCTM_Break, time=0, value=0.4}
        {arriveTangent=-0.19, tangentMode=RCTM_User, time=0.9, value=1}
       ]}}})
   &ScaleColor(
    Linear Color Curve=DataInterface() {propertyValues {
      AlphaCurve {keys [<at 0 0>,<at 0.2933 0.2>,<at 0.8577 1>]}
      BlueCurve {keys [<at 0 1>,<at 0.6316 0>]}
      GreenCurve {keys [<at 0 1>,<at 0.6316 0.7551>]}
      RedCurve {keys [<at 0 1>,<at 0.6316 1>]}
     }}
    Scale Mode=::NewEnumerator2
   )
   /SetVariables(
    Particles.GlowColor=&ColorFromCurve(ColorCurve=DataInterface() {propertyValues {
       AlphaCurve {keys [<at 0 0>,<at 0.1904 0.8>,<at 0.8577 1>]}
       BlueCurve {keys [<at 0 0>,<at 0.645 0>]}
       GreenCurve {keys [<at 0 0.6815>,<at 0.645 0.971>]}
       RedCurve {keys [<at 0 1>,<at 0.645 1.5>]}
      }})
    Particles.GlowSize=&Vector2DFromFloat(Value=30)
   ) {isSetParameters=true}
   &SolveForcesAndVelocity(Clamp Velocity=false, Limit Acceleration=false)
  ]
  rendererClasses [&NiagaraSpriteRendererProperties]
  renderers [
   {rendererClass=&NiagaraSpriteRendererProperties}
   {rendererClass=&NiagaraSpriteRendererProperties, rendererIndex=1}
   {rendererClass=&NiagaraSpriteRendererProperties, rendererIndex=2}
  ]
 }
 ChestSmallParticle {
  emitterUpdateScript [
   &EmitterState()
   &SpawnRate(SpawnRate=35, Use Spawn Probability=false)
  ]
  particleSpawnScript [
   &InitializeParticle(
    Lifetime Max=1.2
    Lifetime Min=0.6
    Lifetime Mode=::NewEnumerator1
    Position Offset=<16 0 17>
    Sprite Size Mode=::NewEnumerator1
    Uniform Sprite Size Max=2
    Uniform Sprite Size Min=0.4
    UsePositionOffset=true
    Write Material Random=true
   )
   &ShapeLocation(MaskSpawnGroup=false, Non Uniform Scale=<0.4 1 1>, Override Seed=false, Sphere Radius=10, UseOwnerScale=true)
   &AddVelocity(
    Module.Use Falloff Distance=false
    Origin Offset=<15 0 17>
    Override Seed=false
    Velocity Mode=::NewEnumerator1
    Velocity Speed=&RandomRangeFloat(Minimum=65, Maximum=140)
   )
  ]
  particleUpdateScript [
   &Drag(Drag=3, Use Linear Drag=true, Use Rotational Drag=true)
   &GravityForce(
    Coordinate Space=::Local
    Gravity=&MakeVector(
     X=&FloatFromCurve(
      FloatCurve=DataInterface() {propertyValues {Curve {keys [
          {arriveTangent=4.1785, interpMode=RCIM_Cubic, leaveTangent=4.1785, tangentMode=RCTM_User, time=0, value=0}
          {arriveTangent=1.1179, interpMode=RCIM_Cubic, leaveTangent=1.1179, tangentMode=RCTM_User, time=0.2042, value=0.6228}
          {arriveTangent=0.0224, interpMode=RCIM_Cubic, leaveTangent=0.0224, tangentMode=RCTM_User, time=0.8, value=1}
         ]}}}
      Scale Curve=-460
     )
     Y=0
     Z=0
    )
   )
   &SolveForcesAndVelocity(Clamp Velocity=false, Limit Acceleration=false)
   &ParticleState()
   &ScaleSpriteSizeBySpeed(Max Scale Factor=<1 3>, Min Scale Factor=<0.6 1.4>)
   &ScaleColor(
    Linear Color Curve=DataInterface() {propertyValues {
      AlphaCurve {keys [<at 0.1953 1>,<at 1 0>]}
      BlueCurve {keys [<at 0 0.7761>,<at 0.2554 0>,<at 0.5805 10>,<at 0.9869 0.7344>]}
      GreenCurve {keys [<at 0 4.9045>,<at 0.2554 0.5588>,<at 0.5805 6.8503>,<at 0.9869 0.3061>]}
      RedCurve {keys [<at 0 10>,<at 0.2554 1>,<at 0.5805 0>,<at 0.9869 0>]}
     }}
    Scale Mode=::NewEnumerator2
   )
  ]
  rendererClasses [&NiagaraSpriteRendererProperties]
  renderers [{rendererClass=&NiagaraSpriteRendererProperties}]
 }
 Decal {
  emitterUpdateScript [
   &EmitterState()
   &SpawnBurst_Instantaneous(Use Spawn Probability=false, UseLoopCountLimit=false)
  ]
  particleSpawnScript [
   &InitializeParticle(
    Color=<rgba 0.2874 0.7682 1.25 1>
    Lifetime=1.4
    Mesh Scale Mode=::NewEnumerator3
    Mesh Uniform Scale=2
    Position Offset=<0 0 22>
    Sprite Size Mode=::NewEnumerator3
    Uniform Sprite Size=140
    UsePositionOffset=true
    Write Material Random=true
   )
   &CameraOffset(Camera Offset Amount=30)
   &DynamicMaterialParameters(
    Index 0 Param 1=1
    Index 0 Param 3=1
    Module.Index0Param0Write=true
    Module.Index0Param1Write=true
    Module.Index0Param2Write=true
    Module.Index0Param3Write=true
   )
   &ApplyInitialForces()
  ]
  particleUpdateScript [
   &ParticleState()
   &ScaleSpriteSize(Uniform Curve Sprite Scale=DataInterface() {propertyValues {Curve {keys [{interpMode=RCIM_Cubic, leaveTangent=-0.5445, tangentMode=RCTM_Break, time=0, value=0}, {tangentMode=RCTM_User, time=1, value=1}]}}})
   &Drag(Drag=3, Use Linear Drag=true, Use Rotational Drag=true)
   &SolveForcesAndVelocity(Clamp Velocity=false, Limit Acceleration=false)
   &ScaleColor(
    Linear Color Curve=DataInterface() {propertyValues {
      AlphaCurve {keys [<at 0.0723 0>,<at 0.4674 1.2>,<at 0.842 2>]}
      BlueCurve {keys [<at 0 1>,<at 1 1>]}
      GreenCurve {keys [<at 0 1>,<at 1 1>]}
      RedCurve {keys [<at 0 1>,<at 1 1>]}
     }}
    Scale Mode=::NewEnumerator2
   )
  ]
  rendererClasses [&NiagaraSpriteRendererProperties]
  renderers [{rendererClass=&NiagaraSpriteRendererProperties}]
 }
 Embers {
  emitterUpdateScript [
   &EmitterState()
   &SpawnRate(SpawnRate=40, Use Spawn Probability=false)
  ]
  particleSpawnScript [
   &PartitionParticles(Number Of Partitions=4)
   /SetVariables(Particles.Random size:NiagaraFloat=&RandomRangeFloat(Minimum=0.1, Maximum=0.2)) {isSetParameters=true}
   &InitializeParticle(
    Color Mode=::NewEnumerator0
    Lifetime Max=0.7
    Lifetime Min=0.55
    Lifetime Mode=::NewEnumerator1
    Sprite Rotation Mode=::NewEnumerator2
    Sprite UV Mode=::NewEnumerator3
    UsePositionOffset=false
    Write Material Random=true
   ) {module=InitializeParticle001}
   &DynamicMaterialParameters(
    Index 0 Param 2=&RandomRangeFloat(Minimum=0, Maximum=4)
    Index 0 Param 4=2.5
    Module.Param0WriteEnabled=true
    Module.Param1WriteEnabled=true
    Module.Param2WriteEnabled=false
    Module.Param3WriteEnabled=true
   ) {module=DynamicMaterialParameters001}
   &SkeletalMeshLocation(
    Bone Sampling Mode=::NewEnumerator0
    MaskSpawnGroup=false
    OverrideRandomSeed=false
    Position Sampling=::NewEnumerator1
    Sampled Position Offset=<-7 0 0>
    Skeletal Mesh=$System.Skeletal Mesh
   )
   &ShapeLocation(MaskSpawnGroup=false, Override Seed=false, Sphere Radius=15, UseOwnerScale=true)
   &AddVelocity(Velocity=&RandomRangeVector(Minimum=<-250 0 0>, Maximum=<-400 0 0>)) {module=AddVelocity001}
   &CameraOffset(Camera Offset Amount=-50) {module=CameraOffset001}
   /SetVariables() {isSetParameters=true}
   /SetVariables(Particles.DY3:Vector4f=&MakeVector4(X=3, Y=1, Z=1, W=1)) {isSetParameters=true}
   /SetVariables(Particles.Position=&AddVectorToPosition(Position=$Particles.Position, Vector=<0 0 15>)) {isSetParameters=true}
  ]
  particleUpdateScript [
   /SetVariables(Particles.Size Scale:NiagaraFloat=&FloatFromCurve(FloatCurve=DataInterface() {propertyValues {Curve {keys [<at 0 0>,<at 0.1 0.5>,<at 1 0>]}}})) {isSetParameters=true}
   &SubUVAnimation(End Frame=11, Start Frame=8, SubUV Animation Mode:ENiagara_SubUVLookupMode=::NewEnumerator2)
   /Niagara/Modules/Update/Forces/CurlNoiseForce(Noise Frequency=100, Noise Strength=&Multiply_Float(A=50, B=100), Pan Noise=true) {enabled=false}
   &ParticleState()
   &ScaleSpriteSize(Scale Sprite Size Mode=::NewEnumerator0, Uniform Scale Factor=$Particles.Size Scale) {module=ScaleSpriteSize001}
   &SpriteRotationRate(Rotation Rate=&RandomRangeFloat(Minimum=-5, Maximum=5)) {module=SpriteRotationRate001}
   &Drag(Drag=4, Use Linear Drag=true, Use Rotational Drag=true)
   &SolveForcesAndVelocity(Clamp Velocity=false, Limit Acceleration=false)
   &FN_Sprite_Attributes(
    Color=&MakeLinearColorFromVectorAndFloat(
     "Vector (RGB)"=&MakeVectorFromLinearColorRGB(Color=&ColorFromCurve(ColorCurve=DataInterface() {propertyValues {
         AlphaCurve {keys []}
         BlueCurve {keys [<at 0.0072 10>, {arriveTangent=-7.051, interpMode=RCIM_Cubic, leaveTangent=-7.051, time=0.5363, value=3}, <at 1 3 RCIM_Cubic>]}
         GreenCurve {keys [<at 0.0072 7.8849>, {arriveTangent=-7.347, interpMode=RCIM_Cubic, leaveTangent=-7.347, time=0.5363, value=2.0848}, <at 1 0.591 RCIM_Cubic>]}
         RedCurve {keys [<at 0.0072 1.887>, {arriveTangent=-1.9008, interpMode=RCIM_Cubic, leaveTangent=-1.9008, time=0.5363, value=0.3592}, <at 1 0 RCIM_Cubic>]}
        }}))
     "Float (Alpha)"=2
    )
    Create Color Attribute=true
    Create SpriteSize Attribute=true
    SpriteSize=&Vector2DFromFloat(Value=&Multiply_Float(A=100, B=&Multiply_Float(A=$Particles.Random size, B=$Particles.Size Scale)))
    Write to Module Namespace=true
   )
   &FN_Sprite_Attributes(
    Color=&MakeLinearColorFromVectorAndFloat(
     "Vector (RGB)"=&MakeVectorFromLinearColorRGB(Color=<rgba 0 0.6744 1.5 1>)
     "Float (Alpha)"=&Multiply_Float(
      A=&FN_FrontEndFloatModifier(Gameplay=0.3, Front End=0.2)
      B=&FloatFromCurve(
       FloatCurve=DataInterface() {propertyValues {Curve {keys [{interpMode=RCIM_Cubic, tangentMode=RCTM_User, time=0, value=1}, {interpMode=RCIM_Cubic, tangentMode=RCTM_User, time=1, value=1}]}}}
       Scale Curve=0.5
      )
     )
    )
    Create Color Attribute=true
    Create SpriteSize Attribute=true
    SpriteSize=&Vector2DFromFloat(Value=&Multiply_Float(A=350, B=&Multiply_Float(A=$Particles.Random size, B=$Particles.Size Scale)))
    Write to Module Namespace=true
   ) {module=FN_Sprite_Attributes001}
  ]
  rendererClasses [&NiagaraSpriteRendererProperties]
  renderers [
   {rendererClass=&NiagaraSpriteRendererProperties}
   {rendererClass=&NiagaraSpriteRendererProperties, rendererIndex=1}
   {rendererClass=&NiagaraSpriteRendererProperties, rendererIndex=2}
  ]
 }
 EyesTrail {
  emitterSpawnScript [/SetVariables(
    Emitter.FlameSpeed:NiagaraFloat=&RandomRangeFloat(Minimum=2, Maximum=4)
    Emitter.UV Offset:NiagaraFloat=&RandomRangeFloat(Minimum=0.11, Maximum=0.99)
   ) {isSetParameters=true}]
  emitterUpdateScript [
   &EmitterState()
   &SpawnBurst_Instantaneous(Spawn Count=2, Use Spawn Probability=false, UseLoopCountLimit=false)
  ]
  particleSpawnScript [
   &PartitionParticles()
   &InitializeParticle(
    Color=<rgba 0.6421 1.6347 3 0.6>
    Lifetime=2
    Mesh Scale=<1.3 4 2>
    Mesh Scale Mode=::NewEnumerator4
    Sprite Size Mode=::NewEnumerator3
    Uniform Sprite Size=6
    UsePositionOffset=false
    Write Material Random=true
   )
   &InitialMeshOrientation(
    Mesh Axis To Orient=<0 1 0>
    Mesh Orientation Mode=::NewEnumerator2
    Orientation Vector=<0 1 0>
    Rotation=&MaskVectorBySpawnGroup(Masked Vector=<0 0.8394 0>, Passthrough Vector=<0.01 0.25 0>, Spawn Group Mask=$Particles.PartitionParticles.Partition)
    Use Rotation Vector=true
   )
  ]
  particleUpdateScript [
   &ParticleState()
   &SkeletalMeshLocation(
    Bone / Socket Index=9
    Bone Sampling Mode=::NewEnumerator4
    MaskSpawnGroup=false
    Sampled Position Offset=&SelectVectorFromArray(
     Direct Array Index=$Particles.PartitionParticles.Partition
     Vector Selection Array=DataInterface() {propertyValues {FloatData [<4.6 9 2.75>, <4.6 9 -2.75>]}}
    )
    Skeletal Mesh=$System.Skeletal Mesh
   )
   &ScaleColor(
    Linear Color Curve=DataInterface() {propertyValues {
      AlphaCurve {keys [<at 0.1987 0>,<at 0.6672 1>,<at 0.7981 1>,<at 1 0>]}
      BlueCurve {keys [<at 0 1>,<at 0.8464 1.5>]}
      GreenCurve {keys [<at 0 1>,<at 0.8464 0.9445>]}
      RedCurve {keys [<at 0 1>,<at 0.8464 0.2001>]}
     }}
    Scale Mode=::NewEnumerator2
   )
   &Drag(Drag=3, Use Linear Drag=true, Use Rotational Drag=true)
   /Niagara/Modules/Update/Forces/V2/CurlNoiseForce(Bias Noise=false, Noise Frequency=20, Noise Strength=200, Pan Noise=false) {enabled=false}
   &SolveForcesAndVelocity(Clamp Velocity=false, Limit Acceleration=false)
   /SetVariables(Particles.EyeTrailColor:LinearColor=&MultiplyLinearColor(
     Linear Color=&ColorFromCurve(ColorCurve=DataInterface() {propertyValues {
        AlphaCurve {keys [<at 0.1996 0>,<at 0.502 1>,<at 0.7981 1.4>,<at 1 0>]}
        BlueCurve {keys [<at 0.2293 6>,<at 0.8464 2>]}
        GreenCurve {keys [<at 0.2293 1.9145>,<at 0.8464 0.5637>]}
        RedCurve {keys [<at 0.2293 0>,<at 0.8464 0.0849>]}
       }})
     Scale Factor=<1 1 1 2.2>
    )) {isSetParameters=true}
  ]
  rendererClasses [&NiagaraSpriteRendererProperties, &NiagaraMeshRendererProperties]
  renderers [
   {rendererClass=&NiagaraSpriteRendererProperties}
   {rendererClass=&NiagaraMeshRendererProperties, rendererIndex=1}
  ]
 }
 HeadTrails {
  emitterUpdateScript [
   &EmitterState()
   &SpawnRate(SpawnRate=10, Use Spawn Probability=false)
  ]
  particleSpawnScript [
   &InitializeParticle(
    Color=<rgba 6.5562 9.8009 10 1>
    Lifetime Max=0.5
    Lifetime Min=0.3
    Lifetime Mode=::NewEnumerator1
    Mesh Scale=&RandomRangeVector(Minimum=<1 1 0.1>, Maximum=<1 1 0.25>)
    Mesh Scale Mode=::NewEnumerator4
    Position Offset=&RandomRangeVector(Minimum=<5 5 50>, Maximum=<-5 -5 50>)
    Sprite Size Mode=::NewEnumerator3
    Uniform Sprite Size=1
    UsePositionOffset=false
    Write Material Random=true
   )
   &InitialMeshOrientation(Mesh Orientation Mode=::NewEnumerator7, Rotation=<0 -0.25 0>, Use Rotation Vector=true)
   &ShapeLocation(
    Hemisphere Distribution=<0.5 1>
    MaskSpawnGroup=false
    Offset=<11 0 48>
    Offset Mode=::NewEnumerator0
    Override Seed=false
    Rotation Angle=0.25
    Rotation Axis=<0 0 1>
    Rotation Mode=::NewEnumerator1
    Sphere Radius=6.5
    UseOwnerScale=true
   )
  ]
  particleUpdateScript [
   &ParticleState()
   &ScaleMeshSize(Scale Factor=&VectorFromCurve(VectorCurve=DataInterface() {propertyValues {XCurve {keys [<at 0 1>]}, YCurve {keys [<at 0 1>]}, ZCurve {keys [<at 0 0.3>,<at 0.2 0.75>,<at 1 1.2>]}}}))
   &DynamicMaterialParameters(
    Index 0 Param 1=1
    Index 0 Param 2=&RandomRangeFloat(Minimum=-5, Maximum=10)
    Module.Index0Param0Write=true
    Module.Index0Param1Write=true
    Module.Index0Param2Write=true
    Module.Index0Param3Write=true
   )
   &ScaleColor(
    Scale Alpha=&FloatFromCurve(FloatCurve=DataInterface() {propertyValues {Curve {keys [{interpMode=RCIM_Cubic, leaveTangent=-2.0227, tangentMode=RCTM_Break, time=0, value=1}, <at 1 0>]}}})
    Scale RGB=<1 1 2>
    ScaleA=true
    ScaleRGB=true
   )
  ]
  rendererClasses [&NiagaraMeshRendererProperties]
  renderers [{rendererClass=&NiagaraMeshRendererProperties, rendererIndex=0}]
 }
 RibbonsChest {
  emitterUpdateScript [
   &EmitterState()
   &SpawnRate(SpawnRate=5, Use Spawn Probability=false)
  ]
  particleSpawnScript [
   &InitializeParticle(
    AdjustAlpha=false
    AdjustHue=true
    AdjustSaturation=false
    AdjustValue=true
    Color=<rgba 0.154 0.4642 1 1>
    Color Mode=::NewEnumerator3
    Lifetime=&Multiply_Float(
     A=&RandomRangeFloat(Minimum=0.25, Maximum=0.4)
     B=&FloatFromCurve(
      FloatCurve=DataInterface() {propertyValues {Curve {keys [{interpMode=RCIM_Cubic, tangentMode=RCTM_Break, time=0, value=4}, <at 0.6 2.5>]}}}
      CurveIndex=$Emitter.NormalizedLoopAge
     )
    )
    Mesh Renderer Array Visibility Mode=::NewEnumerator3
    Mesh Renderer Info=DataInterface() {propertyValues {MeshRenderer=&self:RibbonsChest_2.NiagaraMeshRendererProperties_0}}
    Mesh Scale=&Multiply_Vector(
     A=&RandomRangeVector(Minimum=<1 1 1>, Maximum=<1.4 1 1>)
     B=&SelectVectorFromArray(
      Array Sampling Mode=::NewEnumerator0
      Direct Array Index=?
      Vector Selection Array=DataInterface() {propertyValues {FloatData [<1 -1 1>, <1 1 1>]}}
      Random Seed=0
     )
    )
    Mesh Scale Mode=::NewEnumerator4
    UsePositionOffset=true
    Value Range:Vector2f=<0.8 4>
    Write Material Random=true
   )
   &InitialMeshOrientation(Mesh Orientation Mode=::NewEnumerator7, Use Rotation Vector=true) {enabled=false}
   /SetVariables(Particles.GlowColor=<rgba 0.0307 0.6644 1 0.25>) {isSetParameters=true}
  ]
  particleUpdateScript [
   &ParticleState()
   &SkeletalMeshLocation(
    MaskSpawnGroup=false
    Mesh Sampling Type=::NewEnumerator5
    Orientation Sampling=::NewEnumerator1
    Sampled Position Offset=<15 0 8>
    Sampled Position Offset Coordinate Space:ENiagaraSkelMeshTransforms=::NewEnumerator2
    Skeletal Mesh=$System.Skeletal Mesh
    Socket Sampling Mode=::NewEnumerator5
   )
   &ScaleMeshSize(Scale Factor=&VectorFromCurve(VectorCurve=DataInterface() {propertyValues {
       XCurve {keys [
         {interpMode=RCIM_Cubic, leaveTangent=0.3832, tangentMode=RCTM_Break, time=0, value=0.2}
         {arriveTangent=3.8261, tangentMode=RCTM_User, time=0.2, value=0.7}
         <at 1 1.4>
        ]}
       YCurve {keys [<at 0 1>]}
       ZCurve {keys [<at 1 1>]}
      }}))
   &DynamicMaterialParameters(
    Index 0 Param 1=1
    Index 0 Param 2=&RandomRangeFloat(Minimum=-5, Maximum=10)
    Module.Index0Param0Write=true
    Module.Index0Param1Write=true
    Module.Index0Param2Write=true
    Module.Index0Param3Write=true
   )
   &ScaleColor(
    Scale Alpha=&FloatFromCurve(FloatCurve=DataInterface() {propertyValues {Curve {keys [
         {arriveTangent=-0.0573, interpMode=RCIM_Cubic, leaveTangent=-0.0573, tangentMode=RCTM_User, time=0, value=0}
         {arriveTangent=3.6332, interpMode=RCIM_Cubic, leaveTangent=-3.4431, tangentMode=RCTM_Break, time=0.4, value=0.95}
         {arriveTangent=-0.0851, tangentMode=RCTM_User, time=1, value=0}
        ]}}})
    Scale RGB=<1 1 2>
    ScaleA=true
    ScaleRGB=true
   )
   /SetVariables(Particles.GlowColor=&Multiply_LinearColorByFloat(
     Linear Color=$Particles.Initial.GlowColor
     Float=&FloatFromCurve(FloatCurve=DataInterface() {propertyValues {Curve {keys [
          {arriveTangent=1.2596, interpMode=RCIM_Cubic, leaveTangent=1.2596, tangentMode=RCTM_User, time=0, value=0}
          {arriveTangent=3.6332, interpMode=RCIM_Cubic, leaveTangent=-3.4431, tangentMode=RCTM_Break, time=0.4, value=0.95}
          {arriveTangent=-0.0851, tangentMode=RCTM_User, time=1, value=0}
         ]}}})
    )) {isSetParameters=true}
  ]
  rendererClasses [&NiagaraMeshRendererProperties]
  renderers [
   {rendererClass=&NiagaraMeshRendererProperties, rendererIndex=0}
   {rendererClass=&NiagaraMeshRendererProperties, rendererIndex=1}
  ]
 }
 RibbonsFullBody {
  emitterUpdateScript [
   &EmitterState()
   &SpawnRate(
    Spawn Probability=&FloatFromCurve(FloatCurve=DataInterface() {propertyValues {Curve {keys [<at 0 0>,<at 0.1 0>,<at 0.12 1.1>]}}}, CurveIndex=$Emitter.NormalizedLoopAge)
    SpawnRate=16
    Use Spawn Probability=true
   )
  ]
  particleSpawnScript [
   &InitializeParticle(
    AdjustAlpha=false
    AdjustHue=true
    AdjustSaturation=false
    AdjustValue=true
    Color=<rgba 0.154 0.4642 1 1>
    Color Mode=::NewEnumerator3
    Lifetime Max=0.65
    Lifetime Min=0.4
    Lifetime Mode=::NewEnumerator1
    Mesh Scale=&RandomRangeVector(Minimum=<1 1 0.2>, Maximum=<1 1 0.6>)
    Mesh Scale Mode=::NewEnumerator4
    Position Offset=<20 0 15>
    UsePositionOffset=true
    Value Range:Vector2f=<0.8 4>
    Write Material Random=true
   )
   &InitialMeshOrientation(Mesh Orientation Mode=::NewEnumerator7, Rotation=<0 -0.25 0>, Use Rotation Vector=true)
   /SetVariables(Particles.GlowColor=<rgba 0.0307 0.6644 1 0.25>) {isSetParameters=true}
  ]
  particleUpdateScript [
   &SkeletalMeshLocation(
    Bone Sampling Mode=::NewEnumerator0
    MaskSpawnGroup=false
    Orientation Sampling=::NewEnumerator1
    OverrideRandomSeed=false
    Skeletal Mesh=$System.Skeletal Mesh
   )
   &ParticleState()
   &ScaleMeshSize(Scale Factor=&VectorFromCurve(VectorCurve=DataInterface() {propertyValues {XCurve {keys [<at 0 1>]}, YCurve {keys [<at 0 1>]}, ZCurve {keys [<at 0 0.3>,<at 0.2 0.75>,<at 1 1.2>]}}}))
   &DynamicMaterialParameters(
    Index 0 Param 1=1
    Index 0 Param 2=&RandomRangeFloat(Minimum=-5, Maximum=10)
    Module.Index0Param0Write=true
    Module.Index0Param1Write=true
    Module.Index0Param2Write=true
    Module.Index0Param3Write=true
   )
   &ScaleColor(
    Scale Alpha=&FloatFromCurve(FloatCurve=DataInterface() {propertyValues {Curve {keys [{interpMode=RCIM_Cubic, leaveTangent=-2.0227, tangentMode=RCTM_Break, time=0, value=1}, <at 1 0>]}}})
    Scale RGB=<1 1 2>
    ScaleA=true
    ScaleRGB=true
   )
   /SetVariables(Particles.GlowColor=&Multiply_LinearColorByFloat(
     Linear Color=$Particles.Initial.GlowColor
     Float=&FloatFromCurve(FloatCurve=DataInterface() {propertyValues {Curve {keys [
          {arriveTangent=1.2596, interpMode=RCIM_Cubic, leaveTangent=1.2596, tangentMode=RCTM_User, time=0, value=0}
          {arriveTangent=3.6332, interpMode=RCIM_Cubic, leaveTangent=-3.4431, tangentMode=RCTM_Break, time=0.4, value=0.95}
          {arriveTangent=-0.0851, tangentMode=RCTM_User, time=1, value=0}
         ]}}})
    )) {isSetParameters=true}
  ]
  rendererClasses [&NiagaraMeshRendererProperties]
  renderers [
   {rendererClass=&NiagaraMeshRendererProperties, rendererIndex=0}
   {rendererClass=&NiagaraMeshRendererProperties, rendererIndex=1}
  ]
 }
 SpeedLines {
  emitterUpdateScript [
   &EmitterState()
   &SpawnRate(
    SpawnRate=&FloatFromCurve(
     FloatCurve=DataInterface() {propertyValues {Curve {keys [
         {interpMode=RCIM_Cubic, leaveTangent=2.3031, tangentMode=RCTM_Break, time=0, value=0.4}
         {arriveTangent=-0.2475, tangentMode=RCTM_User, time=0.6, value=1}
        ]}}}
     CurveIndex=$Emitter.NormalizedLoopAge
     Scale Curve=40
    )
    Use Spawn Probability=false
   )
  ]
  particleSpawnScript [
   &InitializeParticle(
    Color=<rgba 1.1544 2.1295 3 1.5>
    Lifetime Max=0.4
    Lifetime Min=0.2
    Lifetime Mode=::NewEnumerator1
    Sprite Size Max=<1.2 50>
    Sprite Size Min=<0.6 15>
    Sprite Size Mode=::NewEnumerator2
    UsePositionOffset=false
    Write Material Random=true
   )
   &SkeletalMeshLocation(MaskSpawnGroup=false, OverrideRandomSeed=false, Skeletal Mesh=$System.Skeletal Mesh) {enabled=false}
   &ShapeLocation(MaskSpawnGroup=false, Override Seed=false, Sphere Radius=70, Sphere Surface Distribution=0.6667, UseOwnerScale=true)
   &AddVelocity(Velocity=&RandomRangeVector(Minimum=<-400 0 0>, Maximum=<-1000 0 0>))
  ]
  particleUpdateScript [
   &ParticleState()
   &ScaleColor(
    Linear Color Curve=DataInterface() {propertyValues {
      AlphaCurve {keys [<at 0 0>,<at 0.1573 1>,<at 0.5937 0.3>,<at 1 0>]}
      BlueCurve {keys [<at 0 1>,<at 0.8464 1>]}
      GreenCurve {keys [<at 0 1>,<at 0.8464 0.6297>]}
      RedCurve {keys [<at 0 1>,<at 0.8464 0.1334>]}
     }}
    Scale Mode=::NewEnumerator2
   )
   &CameraOffset(Camera Offset Amount=25)
   &Drag(Drag=3, Use Linear Drag=true, Use Rotational Drag=true)
   &CurlNoiseForce(Bias Noise=false, Noise Frequency=20, Noise Strength=200, Pan Noise=false) {enabled=false}
   &SolveForcesAndVelocity(Clamp Velocity=false, Limit Acceleration=false)
   &ScaleSpriteSizeBySpeed(Max Scale Factor=<1.2 3>, Min Scale Factor=<1 1.4>)
  ]
  rendererClasses [&NiagaraSpriteRendererProperties]
  renderers [{rendererClass=&NiagaraSpriteRendererProperties}]
 }
}
summary {
 systemName=NS_Emote_FeatherMudLounge_SwapToLion_Loop
 emitters [
  {emitterName=SpeedLines, rendererClasses [&NiagaraSpriteRendererProperties]}
  {emitterName=RibbonsChest, rendererClasses [&NiagaraMeshRendererProperties]}
  {emitterName=HeadTrails, rendererClasses [&NiagaraMeshRendererProperties]}
  {emitterName=EyesTrail, rendererClasses [&NiagaraSpriteRendererProperties, &NiagaraMeshRendererProperties]}
  {emitterName=Decal, rendererClasses [&NiagaraSpriteRendererProperties]}
  {emitterName=RibbonsFullBody, rendererClasses [&NiagaraMeshRendererProperties]}
  {emitterName=Embers, rendererClasses [&NiagaraSpriteRendererProperties]}
  {emitterName=ChestLogo, rendererClasses [&NiagaraSpriteRendererProperties]}
  {emitterName=ChestSmallParticle, rendererClasses [&NiagaraSpriteRendererProperties]}
 ]
 userVariables [
  User.bIsFrontEnd {type=NiagaraBool}
  User.bIsFrontEndPreview {type=NiagaraBool}
 ]
}
