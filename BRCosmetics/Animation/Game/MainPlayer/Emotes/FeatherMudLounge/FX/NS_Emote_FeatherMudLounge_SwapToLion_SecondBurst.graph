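# Niagara system for the second burst of the FeatherMudLounge emote's swap to lion, a short blue-cyan flash.
# Instant bursts emit growing mesh flames, a layered sprite flare with rays, and GPU embers sampled from the character's skeletal mesh.
# A delayed BodyGlow mesh follows a skeletal bone or socket with a vertical offset and fades by curve.
# Glow and ember alpha use separate gameplay and front-end strengths.
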NiagaraSystem /BRCosmetics/Animation/Game/MainPlayer/Emotes/FeatherMudLounge/FX/NS_Emote_FeatherMudLounge_SwapToLion_SecondBurst
dependencies {
 usedDataInterfaces [
  &NiagaraDataInterfaceCurve
  &NiagaraDataInterfaceColorCurve
  &NiagaraDataInterfaceArrayInt32
  &NiagaraDataInterfaceSkeletalMesh
 ]
 usedDynamicInputs [&RandomRangeVector, &InvertBool]
 usedModules [
  &self:SystemSpawnScript.SystemScriptSource.SystemScriptGraph.NiagaraNodeAssignment_1.SetVariables_#1
  &self:SystemSpawnScript.SystemScriptSource.SystemScriptGraph.NiagaraNodeAssignment_0.SetVariables_#2
  &SystemState
 ]
}
emitters {
 BodyGlow {
  emitterUpdateScript [
   &EmitterState(Life Cycle Mode=::NewEnumerator1, Loop Behavior=::NewEnumerator1, Loop Delay=1.1, Loop Duration=1, UseLoopDelay=true)
   &SpawnBurst_Instantaneous(Use Spawn Probability=false, UseLoopCountLimit=false)
  ]
  particleSpawnScript [&InitializeParticle(
    Color Mode=::NewEnumerator0
    Lifetime=1.6
    Mesh Scale=<0.8 0.8 2.4>
    Mesh Scale Mode=::NewEnumerator4
    UsePositionOffset=false
    Write Material Random=true
   )]
  particleUpdateScript [
   &ParticleState()
   &SkeletalMeshLocation(
    Bone / Socket Index=&SelectIntFromArray(Direct Array Index=0, Int Selection Array=DataInterface() {propertyValues {IntData [6]}})
    MaskSpawnGroup=false
    Mesh Sampling Type=::NewEnumerator5
    Orientation Sampling=::NewEnumerator1
    Sampled Position Offset=<0 0 95>
    Skeletal Mesh=$System.Skeletal Mesh
    Socket Sampling Mode=::NewEnumerator5
   )
   &CameraOffset()
   &ScaleColor(
    Scale Alpha=&FloatFromCurve(
     FloatCurve=DataInterface() {propertyValues {Curve {keys [
         {arriveTangent=9.0725, interpMode=RCIM_Cubic, leaveTangent=9.0725, tangentMode=RCTM_User, time=0, value=0}
         {arriveTangent=-0.737, interpMode=RCIM_Cubic, leaveTangent=0.1165, tangentMode=RCTM_Break, time=0.2, value=0.975}
         {arriveTangent=-2.2978, interpMode=RCIM_Cubic, leaveTangent=-2.2978, tangentMode=RCTM_User, time=1, value=0.025}
        ]}}}
     Scale Curve=&FN_FrontEndFloatModifier(Gameplay=1.5, Front End=2.2)
    )
    ScaleA=true
    ScaleRGB=true
   )
   &ScaleColor(ScaleA=true, ScaleRGB=true) {enabled=false, module=ScaleColor001}
  ]
  rendererClasses [&NiagaraMeshRendererProperties]
  renderers [
   {rendererClass=&NiagaraMeshRendererProperties, rendererIndex=0}
   {rendererClass=&NiagaraMeshRendererProperties, rendererIndex=1}
  ]
 }
 Embers {
  simTarget=GPUComputeSim
  emitterUpdateScript [
   &EmitterState(Life Cycle Mode=::NewEnumerator1, Loop Behavior=::NewEnumerator1, Loop Duration=1, UseLoopDelay=true)
   &SpawnBurst_Instantaneous(Spawn Count=45, Use Spawn Probability=false, UseLoopCountLimit=false)
  ]
  particleSpawnScript [
   &PartitionParticles(Number Of Partitions=4)
   /SetVariables(Particles.Random size:NiagaraFloat=&RandomRangeFloat(Minimum=0.1, Maximum=0.2)) {isSetParameters=true}
   &InitializeParticle(
    Color Mode=::NewEnumerator0
    Lifetime Max=0.7
    Lifetime Min=0.55
    Lifetime Mode=::NewEnumerator1
    Sprite Rotation Mode=::NewEnumerator2
    Sprite UV Mode=::NewEnumerator3
    UsePositionOffset=false
    Write Material Random=true
   ) {module=InitializeParticle001}
   &DynamicMaterialParameters(
    Index 0 Param 2=&RandomRangeFloat(Minimum=0, Maximum=4)
    Index 0 Param 4=2.5
    Module.Param0WriteEnabled=true
    Module.Param1WriteEnabled=true
    Module.Param2WriteEnabled=false
    Module.Param3WriteEnabled=true
   ) {module=DynamicMaterialParameters001}
   &SkeletalMeshLocation(
    Bone Sampling Mode=::NewEnumerator0
    MaskSpawnGroup=false
    OverrideRandomSeed=false
    Position Sampling=::NewEnumerator1
    Sampled Position Offset=<-7 0 0>
    Skeletal Mesh=$System.Skeletal Mesh
   )
   &ShapeLocation(MaskSpawnGroup=false, Override Seed=false, Sphere Radius=3, UseOwnerScale=true)
   &AddVelocity(
    Module.Use Falloff Distance=false
    Override Seed=false
    Velocity Mode=::NewEnumerator1
    Velocity Speed=&RandomRangeFloat(Minimum=100, Maximum=200)
   ) {module=AddVelocity001}
   &CameraOffset(Camera Offset Amount=-50) {module=CameraOffset001}
   /SetVariables() {isSetParameters=true}
   /SetVariables(Particles.DY3:Vector4f=&MakeVector4(X=3, Y=1, Z=1, W=1)) {isSetParameters=true}
   /SetVariables(Particles.Position=&AddVectorToPosition(Position=$Particles.Position, Vector=<0 0 15>)) {isSetParameters=true}
  ]
  particleUpdateScript [
   /SetVariables(Particles.Size Scale:NiagaraFloat=&FloatFromCurve(FloatCurve=DataInterface() {propertyValues {Curve {keys [<at 0 0>,<at 0.1 0.5>,<at 1 0>]}}})) {isSetParameters=true}
   &SubUVAnimation(End Frame=11, Start Frame=8, SubUV Animation Mode:ENiagara_SubUVLookupMode=::NewEnumerator2)
   /Niagara/Modules/Update/Forces/CurlNoiseForce(Noise Frequency=100, Noise Strength=&Multiply_Float(A=50, B=100), Pan Noise=true) {enabled=false}
   &ParticleState()
   &ScaleSpriteSize(Scale Sprite Size Mode=::NewEnumerator0, Uniform Scale Factor=$Particles.Size Scale) {module=ScaleSpriteSize001}
   &SpriteRotationRate(Rotation Rate=&RandomRangeFloat(Minimum=-5, Maximum=5)) {module=SpriteRotationRate001}
   &Drag(Drag=4, Use Linear Drag=true, Use Rotational Drag=true)
   &SolveForcesAndVelocity(Clamp Velocity=false, Limit Acceleration=false)
   &FN_Sprite_Attributes(
    Color=&MakeLinearColorFromVectorAndFloat(
     "Vector (RGB)"=&MakeVectorFromLinearColorRGB(Color=&ColorFromCurve(ColorCurve=DataInterface() {propertyValues {
         AlphaCurve {keys []}
         BlueCurve {keys [<at 0.0072 10>, {arriveTangent=-7.051, interpMode=RCIM_Cubic, leaveTangent=-7.051, time=0.5363, value=3}, <at 1 3 RCIM_Cubic>]}
         GreenCurve {keys [<at 0.0072 7.8849>, {arriveTangent=-7.347, interpMode=RCIM_Cubic, leaveTangent=-7.347, time=0.5363, value=2.0848}, <at 1 0.591 RCIM_Cubic>]}
         RedCurve {keys [<at 0.0072 1.887>, {arriveTangent=-1.9008, interpMode=RCIM_Cubic, leaveTangent=-1.9008, time=0.5363, value=0.3592}, <at 1 0 RCIM_Cubic>]}
        }}))
     "Float (Alpha)"=2
    )
    Create Color Attribute=true
    Create SpriteSize Attribute=true
    SpriteSize=&Vector2DFromFloat(Value=&Multiply_Float(A=100, B=&Multiply_Float(A=$Particles.Random size, B=$Particles.Size Scale)))
    Write to Module Namespace=true
   )
   &FN_Sprite_Attributes(
    Color=&MakeLinearColorFromVectorAndFloat(
     "Vector (RGB)"=&MakeVectorFromLinearColorRGB(Color=<rgba 0 0.6744 1.5 1>)
     "Float (Alpha)"=&Multiply_Float(
      A=&FN_FrontEndFloatModifier(Gameplay=0.3, Front End=0.2)
      B=&FloatFromCurve(
       FloatCurve=DataInterface() {propertyValues {Curve {keys [{interpMode=RCIM_Cubic, tangentMode=RCTM_User, time=0, value=1}, {interpMode=RCIM_Cubic, tangentMode=RCTM_User, time=1, value=1}]}}}
       Scale Curve=0.5
      )
     )
    )
    Create Color Attribute=true
    Create SpriteSize Attribute=true
    SpriteSize=&Vector2DFromFloat(Value=&Multiply_Float(A=350, B=&Multiply_Float(A=$Particles.Random size, B=$Particles.Size Scale)))
    Write to Module Namespace=true
   ) {module=FN_Sprite_Attributes001}
  ]
  rendererClasses [&NiagaraSpriteRendererProperties]
  renderers [
   {rendererClass=&NiagaraSpriteRendererProperties}
   {rendererClass=&NiagaraSpriteRendererProperties, rendererIndex=1}
   {rendererClass=&NiagaraSpriteRendererProperties, rendererIndex=2}
  ]
 }
 Flames {
  emitterUpdateScript [
   &EmitterState()
   &SpawnBurst_Instantaneous(Spawn Count=4, Use Spawn Probability=false, UseLoopCountLimit=true)
  ]
  particleSpawnScript [
   &InitializeParticle(
    AdjustAlpha=false
    AdjustHue=true
    AdjustSaturation=false
    AdjustValue=true
    Color=<rgba 0.0077 0.3765 0.8542 1>
    Color Mode=::NewEnumerator3
    Lifetime Max=0.65
    Lifetime Min=0.3
    Lifetime Mode=::NewEnumerator1
    Mesh Scale Mode=::NewEnumerator1
    Mesh Uniform Scale Max=3
    Mesh Uniform Scale Min=1.25
    UsePositionOffset=true
    Value Range:Vector2f=<0.8 5>
    Write Material Random=true
   )
   &InitialMeshOrientation(Use Rotation Vector=false)
   /SetVariables(Particles.GlowColor=<rgba 0.0307 0.6644 1 0.25>) {isSetParameters=true}
  ]
  particleUpdateScript [
   &ParticleState()
   &ScaleMeshSize(Scale Factor=&VectorFromFloat(Value=&FloatFromCurve(FloatCurve=DataInterface() {propertyValues {Curve {keys [<at 0 0.3>,<at 0.2 0.9>,<at 1 1.2>]}}})))
   &DynamicMaterialParameters(
    Index 0 Param 1=1
    Index 0 Param 2=1
    Module.Index0Param0Write=true
    Module.Index0Param1Write=true
    Module.Index0Param2Write=true
    Module.Index0Param3Write=true
   )
   &ScaleColor(
    Scale Alpha=&FloatFromCurve(
     FloatCurve=DataInterface() {propertyValues {Curve {keys [{interpMode=RCIM_Cubic, leaveTangent=-3.2584, tangentMode=RCTM_Break, time=0, value=1}, <at 1 0>]}}}
     Scale Curve=0.7
    )
    Scale RGB=<1 1 2>
    ScaleA=true
    ScaleRGB=true
   )
   /SetVariables(Particles.GlowColor=&Multiply_LinearColorByFloat(
     Linear Color=$Particles.Initial.GlowColor
     Float=&FloatFromCurve(FloatCurve=DataInterface() {propertyValues {Curve {keys [
          {arriveTangent=1.2596, interpMode=RCIM_Cubic, leaveTangent=1.2596, tangentMode=RCTM_User, time=0, value=0.4}
          {arriveTangent=-0.4744, interpMode=RCIM_Cubic, leaveTangent=-6.1012, tangentMode=RCTM_Break, time=0.4, value=1}
          {arriveTangent=-0.0851, tangentMode=RCTM_User, time=1, value=0}
         ]}}})
    )) {isSetParameters=true}
  ]
  rendererClasses [&NiagaraMeshRendererProperties]
  renderers [
   {rendererClass=&NiagaraMeshRendererProperties, rendererIndex=0}
   {rendererClass=&NiagaraMeshRendererProperties, rendererIndex=1}
  ]
 }
 Flare {
  emitterUpdateScript [
   &EmitterState(Life Cycle Mode=::NewEnumerator1, Loop Behavior=::NewEnumerator1, Loop Duration=1, UseLoopDelay=true)
   &SpawnBurst_Instantaneous(Use Spawn Probability=false, UseLoopCountLimit=false)
  ]
  particleSpawnScript [&InitializeParticle(
    Color=<rgba 0.0218 0.0532 0.3177 3>
    Lifetime=0.45
    Position Offset=<0 0 -10>
    Sprite Size Mode=::NewEnumerator3
    Uniform Sprite Size=250
    UsePositionOffset=true
    Write Material Random=true
   )]
  particleUpdateScript [
   &ParticleState()
   &ScaleColor(
    Linear Color Curve=DataInterface() {propertyValues {
      AlphaCurve {keys [<at 0 0>,<at 0.121 1>,<at 0.5151 0.3>,<at 1 0>]}
      BlueCurve {keys [<at 0 1>]}
      GreenCurve {keys [<at 0 1>]}
      RedCurve {keys [<at 0 1>]}
     }}
    Scale Mode=::NewEnumerator2
   )
   &CameraOffset(Camera Offset Amount=-30)
   /SetVariables(
    Particles.FlareColor=&ColorFromCurve(ColorCurve=DataInterface() {propertyValues {
       AlphaCurve {keys [
         <at 0 0>
         {arriveTangent=5.0218, interpMode=RCIM_Cubic, leaveTangent=5.0218, time=0.0591, value=1}
         <at 0.1991 1>
         <at 0.3537 0.25>
         <at 0.7702 0 RCIM_Cubic>
        ]}
       BlueCurve {keys [<at 0 10 RCIM_Cubic>,<at 1 1 RCIM_Cubic>]}
       GreenCurve {keys [<at 0 7.3778 RCIM_Cubic>,<at 1 0.589 RCIM_Cubic>]}
       RedCurve {keys [<at 0 0.4416 RCIM_Cubic>,<at 1 0.0214 RCIM_Cubic>]}
      }})
    Particles.FlareDynPar:Vector4f=<1 0 0 0>
    Particles.FlareSpriteSize:Vector2f=&Vector2DFromFloat(Value=&FloatFromCurve(
      FloatCurve=DataInterface() {propertyValues {Curve {keys [{interpMode=RCIM_Cubic, leaveTangent=2.963, tangentMode=RCTM_Break, time=0, value=0.2}, <at 0.7 1>]}}}
      Scale Curve=320
     ))
   ) {isSetParameters=true}
   /SetVariables(
    Particles.RaysColor:LinearColor=&ColorFromCurve(ColorCurve=DataInterface() {propertyValues {
       AlphaCurve {keys [
         <at 0 0>
         {arriveTangent=0.4832, interpMode=RCIM_Cubic, leaveTangent=0.4832, time=0.0591, value=1}
         <at 0.2719 1>
         <at 0.5174 0.25>
         <at 0.992 0 RCIM_Cubic>
        ]}
       BlueCurve {keys [<at 0 10 RCIM_Cubic>,<at 0.3573 2 RCIM_Cubic>]}
       GreenCurve {keys [<at 0 4.0271 RCIM_Cubic>,<at 0.3573 1.3037 RCIM_Cubic>]}
       RedCurve {keys [<at 0 0 RCIM_Cubic>,<at 0.3573 0.2464 RCIM_Cubic>]}
      }})
    Particles.RaysDynPar:Vector4f=&MakeVector4(
     X=&FloatFromCurve(FloatCurve=DataInterface() {propertyValues {Curve {keys [<at 0 0>,<at 1 1>]}}})
     Y=0
     Z=0
     W=0
    )
    Particles.RaysSpriteSize:Vector2f=&Vector2DFromFloat(Value=270)
   ) {isSetParameters=true}
  ]
  rendererClasses [&NiagaraSpriteRendererProperties]
  renderers [
   {rendererClass=&NiagaraSpriteRendererProperties}
   {rendererClass=&NiagaraSpriteRendererProperties, rendererIndex=1}
   {rendererClass=&NiagaraSpriteRendererProperties, rendererIndex=2}
  ]
 }
}
summary {
 systemName=NS_Emote_FeatherMudLounge_SwapToLion_SecondBurst
 emitters [
  {emitterName=Flames, rendererClasses [&NiagaraMeshRendererProperties]}
  {emitterName=Flare, rendererClasses [&NiagaraSpriteRendererProperties]}
  {emitterName=Embers, simTarget=GPUComputeSim, rendererClasses [&NiagaraSpriteRendererProperties]}
  {emitterName=BodyGlow, rendererClasses [&NiagaraMeshRendererProperties]}
 ]
 userVariables [
  User.bIsFrontEnd {type=NiagaraBool}
  User.bIsFrontEndPreview {type=NiagaraBool}
  User.bIsLocalPlayer {type=NiagaraBool, defaultValue=true}
 ]
}
