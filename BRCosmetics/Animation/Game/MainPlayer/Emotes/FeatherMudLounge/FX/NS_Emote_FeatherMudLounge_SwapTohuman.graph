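# Niagara one-shot burst effect for the FeatherMudLounge emote, playing a blue flame flash for the swap to human.
# DarkGlow adds a glow sprite and fading light, while Flame spawns randomly mirrored, expanding meshes and Embers scatters flickering, shrinking sprites.
# FlareUp, rising ParticleDecal sprite and SpeedLines share System.BurstDelay timing.
# Layered renderers use separate attributes, ember alpha varies between gameplay and front end, and skeletal-mesh spawning is disabled for shape locations.
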NiagaraSystem /BRCosmetics/Animation/Game/MainPlayer/Emotes/FeatherMudLounge/FX/NS_Emote_FeatherMudLounge_SwapTohuman
dependencies {
 usedDataInterfaces [&NiagaraDataInterfaceColorCurve, &NiagaraDataInterfaceCurve]
 usedDynamicInputs []
 usedModules [
  &self:SystemSpawnScript.SystemScriptSource.SystemScriptGraph.NiagaraNodeAssignment_0.SetVariables_#1
  &SystemState
 ]
}
emitters {
 DarkGlow {
  emitterUpdateScript [
   &EmitterState()
   &SpawnBurst_Instantaneous(Use Spawn Probability=false, UseLoopCountLimit=false)
  ]
  particleSpawnScript [&InitializeParticle(
    Color=<rgba 0.0218 0.0532 0.3177 4.5>
    Lifetime=0.4
    Position Offset=<0 0 -20>
    Sprite Size Mode=::NewEnumerator3
    Uniform Sprite Size=250
    UsePositionOffset=true
    Write Material Random=true
   )]
  particleUpdateScript [
   &ParticleState()
   &ScaleColor(
    Linear Color Curve=DataInterface() {propertyValues {
      AlphaCurve {keys [<at 0 0>,<at 0.1573 1>,<at 0.5151 0.3>,<at 1 0>]}
      BlueCurve {keys [<at 0 1>]}
      GreenCurve {keys [<at 0 1>]}
      RedCurve {keys [<at 0 1>]}
     }}
    Scale Mode=::NewEnumerator2
   )
   &CameraOffset(Camera Offset Amount=-30)
   &FN_Light_Attributes(
    Color=<rgba 0.2896 1.0286 2 3>
    Create Color Attribute=true
    Create Exponent Attribute=true
    Create Position Attribute=true
    Create Radius Attribute=true
    Exponent=2.5
    Position=&AddVectorToPosition(Position=$Particles.Initial.Position, Vector=<0 0 -80>)
    Radius=220
    Scale Color=&FloatFromCurve(
     FloatCurve=DataInterface() {propertyValues {Curve {keys [
         {interpMode=RCIM_Cubic, leaveTangent=10.2442, tangentMode=RCTM_Break, time=0, value=0.15}
         {interpMode=RCIM_Cubic, tangentMode=RCTM_Break, time=0.1, value=1}
         <at 1 0>
        ]}}}
     Scale Curve=10
    )
   )
  ]
  rendererClasses [&NiagaraSpriteRendererProperties, &NiagaraLightRendererProperties]
  renderers [
   {rendererClass=&NiagaraSpriteRendererProperties}
   {rendererClass=&NiagaraLightRendererProperties, rendererIndex=1}
  ]
 }
 Embers {
  emitterUpdateScript [
   &EmitterState(Life Cycle Mode=::NewEnumerator1, Loop Behavior=::NewEnumerator1, Loop Duration=1, UseLoopDelay=true)
   &SpawnBurst_Instantaneous(Spawn Count=45, Use Spawn Probability=false, UseLoopCountLimit=false)
  ]
  particleSpawnScript [
   &PartitionParticles(Number Of Partitions=4)
   /SetVariables(Particles.Random size:NiagaraFloat=&RandomRangeFloat(Minimum=0.1, Maximum=0.2)) {isSetParameters=true}
   &InitializeParticle(
    Color Mode=::NewEnumerator0
    Lifetime Max=0.7
    Lifetime Min=0.55
    Lifetime Mode=::NewEnumerator1
    Sprite Rotation Mode=::NewEnumerator2
    Sprite UV Mode=::NewEnumerator3
    UsePositionOffset=false
    Write Material Random=true
   ) {module=InitializeParticle001}
   &DynamicMaterialParameters(
    Index 0 Param 2=&RandomRangeFloat(Minimum=0, Maximum=4)
    Index 0 Param 4=2.5
    Module.Param0WriteEnabled=true
    Module.Param1WriteEnabled=true
    Module.Param2WriteEnabled=false
    Module.Param3WriteEnabled=true
   ) {module=DynamicMaterialParameters001}
   &SkeletalMeshLocation(
    Bone Sampling Mode=::NewEnumerator0
    MaskSpawnGroup=false
    OverrideRandomSeed=false
    Position Sampling=::NewEnumerator1
    Sampled Position Offset=<-7 0 0>
    Skeletal Mesh=$System.Skeletal Mesh
   ) {enabled=false}
   &ShapeLocation(
    Cylinder Height=120
    Cylinder Radius=40
    MaskSpawnGroup=false
    Override Seed=false
    Shape Primitive=::NewEnumerator1
    Surface Only=false
    UseOwnerScale=true
   )
   &AddVelocity(
    Module.Use Falloff Distance=false
    Override Seed=false
    Velocity Mode=::NewEnumerator1
    Velocity Speed=&RandomRangeFloat(Minimum=100, Maximum=200)
   ) {module=AddVelocity001}
   &CameraOffset(Camera Offset Amount=-50) {module=CameraOffset001}
   /SetVariables() {isSetParameters=true}
   /SetVariables(Particles.DY3:Vector4f=&MakeVector4(X=3, Y=1, Z=1, W=1)) {isSetParameters=true}
   /SetVariables(Particles.Position=&AddVectorToPosition(Position=$Particles.Position, Vector=<0 0 15>)) {isSetParameters=true}
  ]
  particleUpdateScript [
   /SetVariables(Particles.Size Scale:NiagaraFloat=&FloatFromCurve(FloatCurve=DataInterface() {propertyValues {Curve {keys [<at 0 0>,<at 0.1 0.5>,<at 1 0>]}}})) {isSetParameters=true}
   &SubUVAnimation(End Frame=11, Start Frame=8, SubUV Animation Mode:ENiagara_SubUVLookupMode=::NewEnumerator2)
   /Niagara/Modules/Update/Forces/CurlNoiseForce(Noise Frequency=100, Noise Strength=&Multiply_Float(A=50, B=100), Pan Noise=true) {enabled=false}
   &ParticleState()
   &ScaleSpriteSize(Scale Sprite Size Mode=::NewEnumerator0, Uniform Scale Factor=$Particles.Size Scale) {module=ScaleSpriteSize001}
   &SpriteRotationRate(Rotation Rate=&RandomRangeFloat(Minimum=-5, Maximum=5)) {module=SpriteRotationRate001}
   &Drag(Drag=4, Use Linear Drag=true, Use Rotational Drag=true)
   &SolveForcesAndVelocity(Clamp Velocity=false, Limit Acceleration=false)
   &FN_Sprite_Attributes(
    Color=&MakeLinearColorFromVectorAndFloat(
     "Vector (RGB)"=&MakeVectorFromLinearColorRGB(Color=&ColorFromCurve(ColorCurve=DataInterface() {propertyValues {
         AlphaCurve {keys []}
         BlueCurve {keys [<at 0.0072 10>, {arriveTangent=-7.051, interpMode=RCIM_Cubic, leaveTangent=-7.051, time=0.5363, value=3}, <at 1 3 RCIM_Cubic>]}
         GreenCurve {keys [<at 0.0072 7.8849>, {arriveTangent=-7.347, interpMode=RCIM_Cubic, leaveTangent=-7.347, time=0.5363, value=2.0848}, <at 1 0.591 RCIM_Cubic>]}
         RedCurve {keys [<at 0.0072 1.887>, {arriveTangent=-1.9008, interpMode=RCIM_Cubic, leaveTangent=-1.9008, time=0.5363, value=0.3592}, <at 1 0 RCIM_Cubic>]}
        }}))
     "Float (Alpha)"=2
    )
    Create Color Attribute=true
    Create SpriteSize Attribute=true
    SpriteSize=&Vector2DFromFloat(Value=&Multiply_Float(A=100, B=&Multiply_Float(A=$Particles.Random size, B=$Particles.Size Scale)))
    Write to Module Namespace=true
   )
   &FN_Sprite_Attributes(
    Color=&MakeLinearColorFromVectorAndFloat(
     "Vector (RGB)"=&MakeVectorFromLinearColorRGB(Color=<rgba 0 0.6744 1.5 1>)
     "Float (Alpha)"=&Multiply_Float(
      A=&FN_FrontEndFloatModifier(Gameplay=0.3, Front End=0.2)
      B=&FloatFromCurve(
       FloatCurve=DataInterface() {propertyValues {Curve {keys [{interpMode=RCIM_Cubic, tangentMode=RCTM_User, time=0, value=1}, {interpMode=RCIM_Cubic, tangentMode=RCTM_User, time=1, value=1}]}}}
       Scale Curve=0.5
      )
     )
    )
    Create Color Attribute=true
    Create SpriteSize Attribute=true
    SpriteSize=&Vector2DFromFloat(Value=&Multiply_Float(A=350, B=&Multiply_Float(A=$Particles.Random size, B=$Particles.Size Scale)))
    Write to Module Namespace=true
   ) {module=FN_Sprite_Attributes001}
  ]
  rendererClasses [&NiagaraSpriteRendererProperties]
  renderers [
   {rendererClass=&NiagaraSpriteRendererProperties}
   {rendererClass=&NiagaraSpriteRendererProperties, rendererIndex=1}
   {rendererClass=&NiagaraSpriteRendererProperties, rendererIndex=2}
  ]
 }
 Flame {
  emitterUpdateScript [
   &EmitterState()
   &SpawnBurst_Instantaneous(Spawn Count=5, Use Spawn Probability=false, UseLoopCountLimit=false)
  ]
  particleSpawnScript [
   &InitializeParticle(
    AdjustAlpha=false
    AdjustHue=true
    AdjustSaturation=false
    AdjustValue=true
    Color=<rgba 0.2338 0.4394 1 1>
    Color Mode=::NewEnumerator3
    Lifetime Max=1.2
    Lifetime Min=0.6
    Lifetime Mode=::NewEnumerator1
    Mesh Scale=&Multiply_Vector(
     A=&RandomRangeVector(Minimum=<1.3 1.3 0.6>, Maximum=<1.3 1.3 1.4>)
     B=&MakeCustomVectorFromBool(Bool=&RandomBool(), True Vector=<1 -1 1>, False Vector=<1 1 1>)
    )
    Mesh Scale Mode=::NewEnumerator4
    Position Offset=<0 0 -80>
    UsePositionOffset=true
    Value Range:Vector2f=<0.8 8>
    Write Material Random=true
   )
   &InitialMeshOrientation(
    Mesh Orientation Mode=::NewEnumerator7
    Rotation=&RandomRangeVector(Minimum=<0 0 0>, Maximum=<0 0 1>)
    Use Rotation Vector=true
   )
   /SetVariables(Particles.GlowColor=<rgba 0.0307 0.6644 1 0.25>) {isSetParameters=true}
  ]
  particleUpdateScript [
   &ParticleState()
   &ScaleMeshSize(Scale Factor=&VectorFromFloat(Value=&FloatFromCurve(FloatCurve=DataInterface() {propertyValues {Curve {keys [<at 0 0.3>,<at 0.2 0.9>,<at 1 1.2>]}}})))
   &DynamicMaterialParameters(
    Index 0 Param 1=1
    Index 0 Param 2=1
    Module.Index0Param0Write=true
    Module.Index0Param1Write=true
    Module.Index0Param2Write=true
    Module.Index0Param3Write=true
   )
   &ScaleColor(
    Scale Alpha=&FloatFromCurve(
     FloatCurve=DataInterface() {propertyValues {Curve {keys [{interpMode=RCIM_Cubic, leaveTangent=-3.2584, tangentMode=RCTM_Break, time=0, value=1}, <at 1 0>]}}}
     Scale Curve=0.7
    )
    Scale RGB=<1 1 2>
    ScaleA=true
    ScaleRGB=true
   )
   /SetVariables(Particles.GlowColor=&Multiply_LinearColorByFloat(
     Linear Color=$Particles.Initial.GlowColor
     Float=&FloatFromCurve(FloatCurve=DataInterface() {propertyValues {Curve {keys [
          {arriveTangent=1.2596, interpMode=RCIM_Cubic, leaveTangent=1.2596, tangentMode=RCTM_User, time=0, value=0.4}
          {arriveTangent=-0.4744, interpMode=RCIM_Cubic, leaveTangent=-3.8835, tangentMode=RCTM_Break, time=0.4, value=1}
          {arriveTangent=-0.0851, tangentMode=RCTM_User, time=1, value=0}
         ]}}})
    )) {isSetParameters=true}
  ]
  rendererClasses [&NiagaraMeshRendererProperties]
  renderers [
   {rendererClass=&NiagaraMeshRendererProperties, rendererIndex=0}
   {rendererClass=&NiagaraMeshRendererProperties, rendererIndex=1}
  ]
 }
 FlareUp {
  emitterUpdateScript [
   &EmitterState()
   &SpawnBurst_Instantaneous(Spawn Count=1, Spawn Time=$System.BurstDelay, Use Spawn Probability=false)
  ]
  particleSpawnScript [
   &InitializeParticle(
    Color=<rgba 0 0.3696 1 1>
    Lifetime=0.35
    Sprite Size Mode=::NewEnumerator3
    Uniform Sprite Size=600
    UsePositionOffset=false
    Write Material Random=true
   )
   &DynamicMaterialParameters(
    Index 0 Param 1=&RandomRangeFloat(Minimum=0.4, Maximum=0.6)
    Index 0 Param 2=1
    Index 0 Param 3=1
    Index 0 Param 4=1
    Module.Param0WriteEnabled=true
    Module.Param1WriteEnabled=true
    Module.Param2WriteEnabled=true
    Module.Param3WriteEnabled=true
   )
   &CameraOffset(Camera Offset Amount=50)
  ]
  particleUpdateScript [
   &SystemLocation(MaskSpawnGroup=false, Offset=&Lerp_Vector(A=<0 0 0>, B=<0 0 100>, Alpha=$Particles.NormalizedAge))
   &ParticleState()
   &SolveForcesAndVelocity(Clamp Velocity=false, Limit Acceleration=false)
   &ScaleColor(
    Scale Alpha=&Multiply_Float(
     A=1
     B=&FloatFromCurve(FloatCurve=DataInterface() {propertyValues {Curve {keys [<at 0 0>,<at 0.2 1>,<at 0.8 1>,<at 1 0>]}}})
    )
    ScaleA=true
    ScaleRGB=true
   )
   &ScaleSpriteSize(Scale Factor=&Vector2DFromFloat(Value=&FloatFromCurve(FloatCurve=DataInterface() {propertyValues {Curve {keys [<at 0 1.5 RCIM_Cubic>,<at 1 0>]}}})))
   /SetVariables(Particles.Flash size:Vector2f=&Multiply_Vector2D_ByFloat(
     Float=&Multiply_Float(
      A=400
      B=&FloatFromCurve(FloatCurve=DataInterface() {propertyValues {Curve {keys [
           {arriveTangent=4.1785, interpMode=RCIM_Cubic, leaveTangent=4.1785, tangentMode=RCTM_User, time=0, value=0}
           {arriveTangent=0.9995, interpMode=RCIM_Cubic, leaveTangent=0.9995, tangentMode=RCTM_User, time=0.1378, value=0.6825}
           {arriveTangent=-0.0405, interpMode=RCIM_Cubic, leaveTangent=-0.0405, tangentMode=RCTM_User, time=0.62, value=1}
          ]}}})
     )
     Vector2D=<1 1>
    )) {isSetParameters=true}
   /SetVariables(Particles.Flash color:LinearColor=&MakeLinearColorFromVectorAndFloat(
     "Vector (RGB)"=&MakeVectorFromLinearColorRGB(Color=<rgba 0.2355 0.9717 1.4 1>)
     "Float (Alpha)"=&Multiply_Float(
      A=&Multiply_Float(
       A=1
       B=&FloatFromCurve(FloatCurve=DataInterface() {propertyValues {Curve {keys [<at 0 0>,<at 0.124 1>,<at 0.496 1>,<at 0.62 0>]}}})
      )
      B=1
     )
    )) {isSetParameters=true}
  ]
  rendererClasses [&NiagaraSpriteRendererProperties]
  renderers [
   {rendererClass=&NiagaraSpriteRendererProperties}
   {rendererClass=&NiagaraSpriteRendererProperties, rendererIndex=1}
  ]
 }
 ParticleDecal {
  emitterUpdateScript [
   &EmitterState()
   &SpawnBurst_Instantaneous(Spawn Time=$System.BurstDelay, Use Spawn Probability=false, UseLoopCountLimit=false)
  ]
  particleSpawnScript [
   &InitializeParticle(
    Color=<rgba 1.8915 4.5507 5 1>
    Lifetime=0.65
    Mesh Scale Mode=::NewEnumerator3
    Mesh Uniform Scale=2
    Position Offset=<0 0 -10>
    Sprite Size Mode=::NewEnumerator3
    Uniform Sprite Size=150
    UsePositionOffset=true
    Write Material Random=true
   )
   &CameraOffset(Camera Offset Amount=30)
   &DynamicMaterialParameters(
    Index 0 Param 1=1
    Index 0 Param 3=1
    Module.Index0Param0Write=true
    Module.Index0Param1Write=true
    Module.Index0Param2Write=true
    Module.Index0Param3Write=true
   )
   &AddVelocity(Velocity=<0 0 550>)
   &ApplyInitialForces()
  ]
  particleUpdateScript [
   &ParticleState()
   &Drag(Drag=3, Use Linear Drag=true, Use Rotational Drag=true)
   &SolveForcesAndVelocity(Clamp Velocity=false, Limit Acceleration=false)
   &ScaleColor(
    Linear Color Curve=DataInterface() {propertyValues {
      AlphaCurve {keys [<at 0.3457 1>,<at 1 0>]}
      BlueCurve {keys [<at 0 1>,<at 1 1>]}
      GreenCurve {keys [<at 0 1>,<at 1 1>]}
      RedCurve {keys [<at 0 1>,<at 1 1>]}
     }}
    Scale Mode=::NewEnumerator2
   )
  ]
  rendererClasses [&NiagaraSpriteRendererProperties]
  renderers [{rendererClass=&NiagaraSpriteRendererProperties}]
 }
 SpeedLines {
  emitterUpdateScript [
   &EmitterState()
   &SpawnBurst_Instantaneous(Spawn Count=12, Spawn Time=$System.BurstDelay, Use Spawn Probability=false, UseLoopCountLimit=false)
  ]
  particleSpawnScript [
   &InitializeParticle(
    Color=<rgba 1.1544 2.1295 3 1.5>
    Lifetime Max=0.5
    Lifetime Min=0.3
    Lifetime Mode=::NewEnumerator1
    Position Offset=<0 0 -80>
    Sprite Size Max=<1.2 50>
    Sprite Size Min=<0.6 15>
    Sprite Size Mode=::NewEnumerator2
    UsePositionOffset=true
    Write Material Random=true
   )
   &SkeletalMeshLocation(MaskSpawnGroup=false, OverrideRandomSeed=false, Skeletal Mesh=$System.Skeletal Mesh) {enabled=false}
   &ShapeLocation(MaskSpawnGroup=false, Override Seed=false, Sphere Radius=70, Sphere Surface Distribution=0.6667, UseOwnerScale=true)
   &AddVelocity(Velocity=&RandomRangeVector(Minimum=<0 0 200>, Maximum=<0 0 500>))
  ]
  particleUpdateScript [
   &ParticleState()
   &ScaleColor(
    Linear Color Curve=DataInterface() {propertyValues {
      AlphaCurve {keys [<at 0 0>,<at 0.1573 1>,<at 0.5937 0.3>,<at 1 0>]}
      BlueCurve {keys [<at 0 1>,<at 0.8464 1>]}
      GreenCurve {keys [<at 0 1>,<at 0.8464 0.6297>]}
      RedCurve {keys [<at 0 1>,<at 0.8464 0.1334>]}
     }}
    Scale Mode=::NewEnumerator2
   )
   &CameraOffset(Camera Offset Amount=25)
   &Drag(Drag=3, Use Linear Drag=true, Use Rotational Drag=true)
   /Niagara/Modules/Update/Forces/V2/CurlNoiseForce(Bias Noise=false, Noise Frequency=20, Noise Strength=200, Pan Noise=false) {enabled=false}
   &SolveForcesAndVelocity(Clamp Velocity=false, Limit Acceleration=false)
   &ScaleSpriteSizeBySpeed(Max Scale Factor=<1.2 3>, Min Scale Factor=<1 1.4>)
  ]
  rendererClasses [&NiagaraSpriteRendererProperties]
  renderers [{rendererClass=&NiagaraSpriteRendererProperties}]
 }
}
summary {
 systemName=NS_Emote_FeatherMudLounge_SwapTohuman
 emitters [
  {emitterName=ParticleDecal, rendererClasses [&NiagaraSpriteRendererProperties]}
  {emitterName=SpeedLines, rendererClasses [&NiagaraSpriteRendererProperties]}
  {emitterName=FlareUp, rendererClasses [&NiagaraSpriteRendererProperties]}
  {emitterName=Flame, rendererClasses [&NiagaraMeshRendererProperties]}
  {emitterName=DarkGlow, rendererClasses [&NiagaraSpriteRendererProperties, &NiagaraLightRendererProperties]}
  {emitterName=Embers, rendererClasses [&NiagaraSpriteRendererProperties]}
 ]
 userVariables [
  User.bIsFrontEnd {type=NiagaraBool}
  User.bIsFrontEndPreview {type=NiagaraBool}
  User.VariantSwitch {type=NiagaraFloat, defaultValue=0, description=""}
 ]
}
